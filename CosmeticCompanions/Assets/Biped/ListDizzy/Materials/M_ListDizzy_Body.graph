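# Cloth-model body material for the ListDizzy cosmetic companion, built on the 2023 anime shader Pre, Calculations and Post functions with a sphere-normal override.
# UV sphere masks paint eye and mouth areas into lit and shaded colors and subtract them from DFL; the eye mask also feeds FX.
# A flipbook face function composites over that color and adds optional World Position Offset.
# A UV1 split offsets each eye oppositely for adjustable spacing.
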Material /CosmeticCompanions/Assets/Biped/ListDizzy/Materials/M_ListDizzy_Body
graph {
 expressions {
  MaterialExpressionAdd_0 {
   class=MaterialExpressionAdd
   inputs [
    A {from=MaterialExpressionGetMaterialAttributes_1, output=World Position Offset}
    B {from=MaterialExpressionLinearInterpolate_3}
   ]
  }
  MaterialExpressionAdd_5 {
   class=MaterialExpressionAdd
   inputs [
    A {from=MaterialExpressionTextureCoordinate_1}
    B {from=MaterialExpressionMultiply_5}
   ]
  }
  MaterialExpressionComponentMask_1 {class=MaterialExpressionComponentMask, inputs [None {from=MaterialExpressionStep_1}], properties {r=true}}
  MaterialExpressionConstant2Vector_0 {class=MaterialExpressionConstant2Vector, properties {g=0.2, r=0.64}}
  MaterialExpressionConstant2Vector_2 {class=MaterialExpressionConstant2Vector, properties {g=0.2, r=0.85}}
  MaterialExpressionConstant2Vector_6 {class=MaterialExpressionConstant2Vector, properties {r=1}}
  MaterialExpressionConstant2Vector_7 {class=MaterialExpressionConstant2Vector, properties {g=0.36, r=0.2455}}
  MaterialExpressionConstant3Vector_0 {class=MaterialExpressionConstant3Vector, properties {constant=<rgba 0.0194 0.159 0.4969 1>}}
  MaterialExpressionConstant3Vector_1 {class=MaterialExpressionConstant3Vector, properties {constant=<rgba 0.008 0.0097 0.013 1>}}
  MaterialExpressionConstant3Vector_3 {class=MaterialExpressionConstant3Vector, properties {constant=<rgba 0.7305 0.7379 0.7529 1>}}
  MaterialExpressionConstant3Vector_4 {class=MaterialExpressionConstant3Vector, properties {constant=<rgba 0.4969 0.5271 0.5457 1>}}
  MaterialExpressionConstant_0 {class=MaterialExpressionConstant}
  MaterialExpressionConvert_0 {
   class=MaterialExpressionConvert
   inputs [None {from=MaterialExpressionMaterialFunctionCall_21, output=FX}]
   properties {
    nodeName=Break Float4
    convertInputs [{
      type=Vector4
      defaultValue=<rgba 0 0 0 1>
      expressionInput {
       expression=&self:MaterialExpressionMaterialFunctionCall_21
       inputName=None
       mask=0
       maskA=0
       maskB=0
       maskG=0
       maskR=0
       outputIndex=3
      }
     }]
    convertOutputs [{type=Scalar}, {type=Scalar}, {type=Scalar}, {type=Scalar}]
   }
  }
  MaterialExpressionConvert_1 {
   class=MaterialExpressionConvert
   inputs [
    None {from=MaterialExpressionConvert_0}
    None {from=MaterialExpressionMax_1}
    None {from=MaterialExpressionConvert_0}
    None {from=MaterialExpressionConvert_0}
   ]
   properties {
    nodeName=Make Vector4
    convertInputs [
     {
      type=Scalar
      expressionInput {
       expression=&self:MaterialExpressionConvert_0
       inputName=None
       mask=0
       maskA=0
       maskB=0
       maskG=0
       maskR=0
       outputIndex=0
      }
     }
     {
      type=Scalar
      expressionInput {
       expression=&self:MaterialExpressionMax_1
       inputName=None
       mask=0
       maskA=0
       maskB=0
       maskG=0
       maskR=0
       outputIndex=0
      }
     }
     {
      type=Scalar
      expressionInput {
       expression=&self:MaterialExpressionConvert_0
       inputName=None
       mask=0
       maskA=0
       maskB=0
       maskG=0
       maskR=0
       outputIndex=2
      }
     }
     {
      type=Scalar
      expressionInput {
       expression=&self:MaterialExpressionConvert_0
       inputName=None
       mask=0
       maskA=0
       maskB=0
       maskG=0
       maskR=0
       outputIndex=3
      }
     }
    ]
    convertOutputs [{type=Vector4, defaultValue=<rgba 0 0 0 1>}]
   }
  }
  MaterialExpressionGetMaterialAttributes_0 {
   class=MaterialExpressionGetMaterialAttributes
   inputs [None {from=MaterialExpressionMaterialFunctionCall_1, output=Material}]
   properties {attributeGetTypes [#1]}
  }
  MaterialExpressionGetMaterialAttributes_1 {
   class=MaterialExpressionGetMaterialAttributes
   inputs [None {from=MaterialExpressionMaterialFunctionCall_14, output=Result}]
   properties {attributeGetTypes [#1]}
  }
  MaterialExpressionLinearInterpolate_0 {
   class=MaterialExpressionLinearInterpolate
   inputs [
    A {from=MaterialExpressionLinearInterpolate_9}
    B {from=MaterialExpressionMaterialFunctionCall_1, output=FaceColor}
    Alpha {from=MaterialExpressionNamedRerouteDeclaration_0}
   ]
  }
  MaterialExpressionLinearInterpolate_1 {
   class=MaterialExpressionLinearInterpolate
   inputs [
    A {from=MaterialExpressionLinearInterpolate_10}
    B {from=MaterialExpressionMultiply_0}
    Alpha {from=MaterialExpressionNamedRerouteDeclaration_0}
   ]
  }
  MaterialExpressionLinearInterpolate_10 {
   class=MaterialExpressionLinearInterpolate
   inputs [
    A {from=MaterialExpressionLinearInterpolate_4}
    B {from=MaterialExpressionConstant3Vector_4}
    Alpha {from=MaterialExpressionMultiply_6}
   ]
   properties {constAlpha=0.1}
  }
  MaterialExpressionLinearInterpolate_2 {
   class=MaterialExpressionLinearInterpolate
   inputs [
    A {from=""}
    B {from=MaterialExpressionScalarParameter_1}
    Alpha {from=MaterialExpressionNamedRerouteUsage_1}
   ]
   properties {constA=0.4}
  }
  MaterialExpressionLinearInterpolate_3 {
   class=MaterialExpressionLinearInterpolate
   inputs [
    A {from=MaterialExpressionConstant_0}
    B {from=MaterialExpressionGetMaterialAttributes_0, output=World Position Offset}
    Alpha {from=MaterialExpressionScalarParameter_2}
   ]
  }
  MaterialExpressionLinearInterpolate_4 {
   class=MaterialExpressionLinearInterpolate
   inputs [
    A {from=MaterialExpressionMaterialFunctionCall_21, output=Color_Shaded}
    B {from=MaterialExpressionConstant3Vector_1}
    Alpha {from=MaterialExpressionMultiply_2}
   ]
   properties {constAlpha=0.1}
  }
  MaterialExpressionLinearInterpolate_6 {
   class=MaterialExpressionLinearInterpolate
   inputs [
    A {from=MaterialExpressionMaterialFunctionCall_21, output=Color_Lit}
    B {from=MaterialExpressionConstant3Vector_0}
    Alpha {from=MaterialExpressionMultiply_2}
   ]
   properties {constAlpha=0.1}
  }
  MaterialExpressionLinearInterpolate_8 {
   class=MaterialExpressionLinearInterpolate
   inputs [
    A {from=MaterialExpressionScalarParameter_4}
    B {from=MaterialExpressionMultiply_4}
    Alpha {from=MaterialExpressionComponentMask_1}
   ]
  }
  MaterialExpressionLinearInterpolate_9 {
   class=MaterialExpressionLinearInterpolate
   inputs [
    A {from=MaterialExpressionLinearInterpolate_6}
    B {from=MaterialExpressionConstant3Vector_3}
    Alpha {from=MaterialExpressionMultiply_6}
   ]
   properties {constAlpha=0.1}
  }
  MaterialExpressionMaterialFunctionCall_0 {
   class=MaterialExpressionMaterialFunctionCall
   inputs [Normal {from=MaterialExpressionMaterialFunctionCall_21, output=Normal}]
   properties {materialFunction=/CosmeticCompanions/Materials/MaterialFunctions/MF_CompanionAnime_NormalSphere}
  }
  MaterialExpressionMaterialFunctionCall_1 {
   class=MaterialExpressionMaterialFunctionCall
   inputs [
    Material {from=""}
    Brow UV {from=""}
    Eye UV {from=MaterialExpressionAdd_5}
    Mouth UV {from=""}
   ]
   properties {materialFunction=/CosmeticCompanions/Materials/MaterialFunctions/MF_FlipbookFace_Companion}
  }
  MaterialExpressionMaterialFunctionCall_14 {
   class=MaterialExpressionMaterialFunctionCall
   inputs [
    Shading_Gradient {from=MaterialExpressionMaterialFunctionCall_20, output=Shading_Gradient}
    Threshold_Gradient {from=MaterialExpressionMaterialFunctionCall_20, output=Shading_Gradient}
    Specular_Gradient {from=MaterialExpressionMaterialFunctionCall_20, output=Shading_Gradient}
    Sheen_Mask {from=MaterialExpressionMaterialFunctionCall_20, output=Shading_Gradient}
    Thickness {from=MaterialExpressionMaterialFunctionCall_20, output=Shading_Gradient}
    DFL_Mask {from=MaterialExpressionMaterialFunctionCall_20, output=Shading_Gradient}
    FX {from=MaterialExpressionMaterialFunctionCall_20, output=Shading_Gradient}
    Metallic_Gradient {from=MaterialExpressionMaterialFunctionCall_20, output=Shading_Gradient}
    RimLight_Gradient {from=MaterialExpressionMaterialFunctionCall_20, output=Shading_Gradient}
    NormalBuffer {from=MaterialExpressionMaterialFunctionCall_20, output=Shading_Gradient}
    -------- {from=""}
    CINE_LightSources {from=MaterialExpressionMaterialFunctionCall_20, output=Shading_Gradient}
    CINE_LightColor {from=MaterialExpressionMaterialFunctionCall_20, output=Shading_Gradient}
    --------- {from=""}
    Color_Lit {from=MaterialExpressionLinearInterpolate_0}
    Color_Shaded {from=MaterialExpressionLinearInterpolate_1}
    SRM {from=MaterialExpressionMaterialFunctionCall_21, output=SRM}
    Emissive {from=MaterialExpressionMaterialFunctionCall_21, output=SRM}
    DFL_Color {from=MaterialExpressionMaterialFunctionCall_21, output=SRM}
    Specular_Color {from=MaterialExpressionMaterialFunctionCall_21, output=SRM}
    RimLight_Color {from=MaterialExpressionMaterialFunctionCall_21, output=SRM}
   ]
   properties {materialFunction=/Game/Global/Materials/Character_Parent_2023/Anime_Parent_2023/MaterialFunctions/MFC_AnimeShader_2023_Post}
  }
  MaterialExpressionMaterialFunctionCall_20 {
   class=MaterialExpressionMaterialFunctionCall
   inputs [
    CustomLightVector {from=""}
    OverrideNormal_WS {from=MaterialExpressionMaterialFunctionCall_0, output=Normal}
    STT {from=MaterialExpressionMaterialFunctionCall_21, output=STT}
    Normal_Map {from=MaterialExpressionMaterialFunctionCall_21, output=STT}
    DFL {from=MaterialExpressionSubtract_0}
    FX {from=MaterialExpressionConvert_1}
   ]
   properties {materialFunction=/Game/Global/Materials/Character_Parent_2023/Anime_Parent_2023/MaterialFunctions/MFC_AnimeShader_2023_Calculations}
  }
  MaterialExpressionMaterialFunctionCall_21 {
   class=MaterialExpressionMaterialFunctionCall
   inputs [CustomUV {from=""}]
   properties {materialFunction=/Game/Global/Materials/Character_Parent_2023/Anime_Parent_2023/MaterialFunctions/MFC_AnimeShader_2023_Pre}
  }
  MaterialExpressionMax_0 {
   class=MaterialExpressionMax
   inputs [
    A {from=MaterialExpressionSphereMask_0}
    B {from=MaterialExpressionSphereMask_1}
   ]
  }
  MaterialExpressionMax_1 {
   class=MaterialExpressionMax
   inputs [
    A {from=MaterialExpressionConvert_0}
    B {from=MaterialExpressionMultiply_2}
   ]
  }
  MaterialExpressionMax_3 {
   class=MaterialExpressionMax
   inputs [
    A {from=MaterialExpressionMultiply_2}
    B {from=MaterialExpressionMultiply_6}
   ]
  }
  MaterialExpressionMultiply_0 {
   class=MaterialExpressionMultiply
   inputs [
    A {from=MaterialExpressionMaterialFunctionCall_1, output=FaceColor}
    B {from=MaterialExpressionScalarParameter_0}
   ]
  }
  MaterialExpressionMultiply_2 {
   class=MaterialExpressionMultiply
   inputs [
    A {from=MaterialExpressionMax_0}
    B {from=MaterialExpressionOneMinus_0}
   ]
  }
  MaterialExpressionMultiply_4 {
   class=MaterialExpressionMultiply
   inputs [
    A {from=MaterialExpressionScalarParameter_4}
    B {from=""}
   ]
   properties {constB=-1}
  }
  MaterialExpressionMultiply_5 {
   class=MaterialExpressionMultiply
   inputs [
    A {from=MaterialExpressionLinearInterpolate_8}
    B {from=MaterialExpressionConstant2Vector_6}
   ]
  }
  MaterialExpressionMultiply_6 {
   class=MaterialExpressionMultiply
   inputs [
    A {from=MaterialExpressionSphereMask_2}
    B {from=MaterialExpressionOneMinus_1}
   ]
  }
  MaterialExpressionNamedRerouteDeclaration_0 {
   class=MaterialExpressionNamedRerouteDeclaration
   inputs [None {from=MaterialExpressionMaterialFunctionCall_1, output=FaceAlpha}]
   properties {name=FlipbookFaceAlpha, nodeColor=<rgba 0.8281 0 1 1>}
  }
  MaterialExpressionNamedRerouteUsage_1 {class=MaterialExpressionNamedRerouteUsage}
  MaterialExpressionOneMinus_0 {class=MaterialExpressionOneMinus, inputs [None {from=MaterialExpressionScalarParameter_3}]}
  MaterialExpressionOneMinus_1 {class=MaterialExpressionOneMinus, inputs [None {from=MaterialExpressionScalarParameter_5}]}
  MaterialExpressionScalarParameter_0 {class=MaterialExpressionScalarParameter, properties {defaultValue=1, parameterName=FlipbookFaceAnimeShadedValue}}
  MaterialExpressionScalarParameter_1 {class=MaterialExpressionScalarParameter, properties {parameterName=Temporal Responsiveness}}
  MaterialExpressionScalarParameter_2 {class=MaterialExpressionScalarParameter, properties {defaultValue=1, group=Flipbook 01 Global, parameterName="Use Flipbook Geo (WPO)"}}
  MaterialExpressionScalarParameter_3 {class=MaterialExpressionScalarParameter, properties {group=Flipbook 03 Eyes, parameterName=DeubgShow_PaintEye, sortPriority=0}}
  MaterialExpressionScalarParameter_4 {class=MaterialExpressionScalarParameter, properties {group=Flipbook 03 Eyes, parameterName=Eyes_gap_offset}}
  MaterialExpressionScalarParameter_5 {class=MaterialExpressionScalarParameter, properties {group=Flipbook 04 Mouth, parameterName=DeubgShow_PaintMouth, sortPriority=0}}
  MaterialExpressionSetMaterialAttributes_0 {
   class=MaterialExpressionSetMaterialAttributes
   inputs [
    MaterialAttributes {from=MaterialExpressionGetMaterialAttributes_1, output=MaterialAttributes}
    World Position Offset {from=MaterialExpressionAdd_0, output=""}
   ]
   properties {attributeSetTypes [#1]}
  }
  MaterialExpressionSphereMask_0 {
   class=MaterialExpressionSphereMask
   inputs [
    A {from=MaterialExpressionTextureCoordinate_0, output=""}
    B {from=MaterialExpressionConstant2Vector_2, output=""}
    Radius {from="", output=""}
    Hardness {from="", output=""}
   ]
   properties {attenuationRadius=0.088}
  }
  MaterialExpressionSphereMask_1 {
   class=MaterialExpressionSphereMask
   inputs [
    A {from=MaterialExpressionTextureCoordinate_0, output=""}
    B {from=MaterialExpressionConstant2Vector_0, output=""}
    Radius {from="", output=""}
    Hardness {from="", output=""}
   ]
   properties {attenuationRadius=0.088}
  }
  MaterialExpressionSphereMask_2 {
   class=MaterialExpressionSphereMask
   inputs [
    A {from=MaterialExpressionTextureCoordinate_0, output=""}
    B {from=MaterialExpressionConstant2Vector_7, output=""}
    Radius {from="", output=""}
    Hardness {from="", output=""}
   ]
   properties {attenuationRadius=0.053}
  }
  MaterialExpressionStep_1 {
   class=MaterialExpressionStep
   inputs [
    Y {from=MaterialExpressionTextureCoordinate_1, output=""}
    X {from="", output=""}
   ]
   properties {constX=0.5}
  }
  MaterialExpressionSubtract_0 {
   class=MaterialExpressionSubtract
   inputs [
    A {from=MaterialExpressionMaterialFunctionCall_21, output=DFL}
    B {from=MaterialExpressionMax_3}
   ]
  }
  MaterialExpressionTemporalResponsivenessOutput_0 {class=MaterialExpressionTemporalResponsivenessOutput, inputs [None {from=MaterialExpressionLinearInterpolate_2, output=""}], properties {}}
  MaterialExpressionTextureCoordinate_0 {class=MaterialExpressionTextureCoordinate}
  MaterialExpressionTextureCoordinate_1 {class=MaterialExpressionTextureCoordinate, properties {coordinateIndex=1}}
 }
 properties {
  bAutomaticallySetUsageInEditor=false
  bUseMaterialAttributes=true
  bUsedWithMeshDeformer=true
  bUsedWithMorphTargets=true
  bUsedWithSkeletalMesh=true
  pixelDepthOffsetMode=PDOM_Legacy
  previewMesh=/CosmeticCompanions/Assets/Biped/ListDizzy/Meshes/ListDizzy_Biped
  shadingModel=MSM_Cloth
  usedShadingModels=Cloth
 }
}
