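# Animation Blueprint for the ListDizzy biped companion cosmetic, posing its skeleton from another mesh's animation.
# Retarget Pose From Mesh supplies the base pose through the RTG_ListDizzy IK retargeter.
# A non-looping additive tail pose test sequence layers on top, weighted by TailAdditiveAlpha, which this Blueprint never sets.
# The Copy Pose From Mesh node is unconnected, and the event graph holds only unwired default nodes.
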AnimBlueprint /CosmeticCompanions/Assets/Biped/ListDizzy/Meshes/ListDizzy_Biped_AnimBP
target_skeleton=&/ListDizzy_Biped_Skeleton
graphs [
 AnimGraph {nodes [
   Copy Pose From Mesh() {bCopyCurves=true, bUseAttachedParent=true}
   Output Pose(Result=Apply Additive(
     Base=Retarget Pose From Mesh!() -> (Pose:Pose Link Structure) {is=title AnimGraphNode_RetargetPoseFromMesh, iKRetargeterAsset=&/RTG_ListDizzy}
     Additive="Sequence Player (additive) 'ListDizzy_COR_Tail_Pose_ADD_Test'"() {bLoopAnimation=false, sequence=/CosmeticCompanions/Assets/Biped/ListDizzy/Animation/Core/ListDizzy_COR_Tail_Pose_ADD_Test}
     Alpha=Get TailAdditiveAlpha() -> (TailAdditiveAlpha:double)
    ))
  ]}
 EventGraph {nodes [
   TryGetPawnOwner()
   Event BlueprintUpdateAnimation()
  ]}
]
variables [TailAdditiveAlpha {type=number}]
